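# Physics asset for the FaceAcc part of the female medium Night Capsule cosmetic body, defining bodies and joints for dynamic hair bones.
# A five-body central hair mass chain and left and right strand chains hang from the head, with an offshoot branching off the left strand.
# Hair joints mostly use angular limits, some locking twist.
# Spine, neck, head, clavicle and upper-arm shapes form an upper-body proxy, linked mostly by free-rotation joints.
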PhysicsAsset /BRCosmetics/Characters/Player/Female/Medium/Bodies/F_MED_Night_Capsule/Meshes/Parts/F_MED_Night_Capsule_FaceAcc_Physics
bodies {
 clavicle_l [{shapeType=Capsule, length=13.5, radius=2.75, shapeName=clavicle_l_capsule, center=<7.503 -1.5943 -0.4879>, rotation=<rot 90 0 0>}]
 clavicle_r [{shapeType=Capsule, length=12.5, radius=4.75, shapeName=clavicle_r_capsule, center=<-1.7362 -0.7749 3.149>, rotation=<rot 90 0 0>}]
 dyn_L_hairStrand_1 [{shapeType=Capsule, length=5, radius=2, shapeName=dyn_L_hairStrand_1_capsule, center=<2.6818 0 0>, rotation=<rot 90 0 0>}]
 dyn_L_hairStrand_2 [{shapeType=Capsule, length=5, radius=1.25, shapeName=dyn_L_hairStrand_2_capsule, center=<1.7853 0 0>, rotation=<rot 90 0 0>}]
 dyn_L_hairStrand_3 [{shapeType=Capsule, length=4.5, radius=1.25, shapeName=dyn_L_hairStrand_3_capsule, center=<2.8355 -0.4 1>, rotation=<rot 65.5565 -180 -180>}]
 dyn_L_hairStrand_offshoot_1 [{shapeType=Capsule, length=4.25, radius=0.75, shapeName=dyn_L_hairStrand_offshoot_1_capsule, center=<1.3385 0 0>, rotation=<rot 90 0 0>}]
 dyn_L_hairStrand_offshoot_2 [{
   shapeType=Capsule
   length=4.5
   radius=1
   shapeName=dyn_L_hairStrand_offshoot_2_capsule
   center=<1.1366 -0.4 0.5>
   rotation=<rot 69.9991 169.9982 -169.9996>
  }]
 dyn_R_hairStrand_1 [{shapeType=Capsule, length=5, radius=2.75, shapeName=dyn_R_hairStrand_1_capsule, center=<3.1884 0 0>, rotation=<rot 90 0 0>}]
 dyn_R_hairStrand_2 [{shapeType=Capsule, length=5, radius=1.75, shapeName=dyn_R_hairStrand_2_capsule, center=<2.7378 0 0>, rotation=<rot 90 0 0>}]
 dyn_R_hairStrand_3 [{shapeType=Capsule, length=5, radius=1.25, shapeName=dyn_R_hairStrand_3_capsule, center=<1.6 -0.9 -0.6>, rotation=<rot 63.8076 -10.0015 0>}]
 dyn_hairMass_1 [{shapeType=Capsule, length=3.75, radius=6.5, shapeName=dyn_hairMass_1_capsule, center=<5 0 0>}]
 dyn_hairMass_2 [{shapeType=Capsule, length=16.75, radius=7, shapeName=dyn_hairMass_2_capsule, center=<5 0 0>}]
 dyn_hairMass_3 [{shapeType=Box, extentX=12.6624, extentY=8.8354, extentZ=16.5965, shapeName=dyn_hairMass_3_box, center=<4.6668 0 0>}]
 dyn_hairMass_4 [{shapeType=Box, extentX=10.1349, extentY=6.8212, extentZ=16.5467, shapeName=dyn_hairMass_4_box, center=<4.4043 1.2 0>}]
 dyn_hairMass_base [{shapeType=Capsule, length=0.505, radius=0.505, shapeName=dyn_hairMass_base_capsule, rotation=<rot 90 20.556 42.3398>}]
 head [{
   shapeType=Box
   extentX=42.5
   extentY=18.8928
   extentZ=64.3036
   shapeName=head_box
   center=<-4.0408 9.3554 -0.0004>
   rotation=<rot 0 9.5042 0>
  }]
 neck_02 [{shapeType=Capsule, length=10, radius=4, shapeName=neck_02_capsule, rotation=<rot 90 0 0>}]
 spine_03 [{shapeType=Capsule, length=12, radius=14, shapeName=spine_03_capsule, center=<5.5804 5.7919 0>}]
 spine_04 [{shapeType=Capsule, length=35.25, radius=8, shapeName=spine_04_capsule, center=<8.8583 -0.376 0>}]
 spine_05 [{shapeType=Capsule, length=16.25, radius=13, shapeName=spine_05_capsule, center=<-0.7967 6.7575 0>}]
 upperarm_l [{shapeType=Capsule, length=15, radius=4.5, shapeName=upperarm_r_capsule, center=<8.2427 0 0>, rotation=<rot -90 0 0>}]
 upperarm_r [{shapeType=Capsule, length=15, radius=4.9, shapeName=upperarm_r_capsule, center=<-8.2427 0 0>, rotation=<rot -90 0 0>}]
}
constraints [
 {bone1Name=dyn_hairMass_1, bone2Name=dyn_hairMass_base, swing1LimitDegrees=80, swing2LimitDegrees=10, swing2Motion=Free, twistLimitDegrees=45}
 {bone1Name=dyn_hairMass_2, bone2Name=dyn_hairMass_1, swing1LimitDegrees=80, swing2LimitDegrees=10, twistLimitDegrees=30}
 {bone1Name=dyn_hairMass_3, bone2Name=dyn_hairMass_2, swing1LimitDegrees=80, swing2LimitDegrees=10, twistLimitDegrees=30}
 {bone1Name=dyn_L_hairStrand_1, bone2Name=head, swing1LimitDegrees=60, swing2LimitDegrees=45, twistLimitDegrees=10}
 {bone1Name=dyn_L_hairStrand_2, bone2Name=dyn_L_hairStrand_1, swing1LimitDegrees=20, swing2LimitDegrees=20, twistLimitDegrees=10}
 {bone1Name=dyn_L_hairStrand_3, bone2Name=dyn_L_hairStrand_2, swing1LimitDegrees=45, swing2LimitDegrees=45, twistLimitDegrees=10}
 {
  bone1Name=dyn_L_hairStrand_offshoot_1
  bone2Name=dyn_L_hairStrand_1
  swing1LimitDegrees=45
  swing2LimitDegrees=45
  twistLimitDegrees=10
  twistMotion=Locked
 }
 {
  bone1Name=dyn_L_hairStrand_offshoot_2
  bone2Name=dyn_L_hairStrand_offshoot_1
  swing1LimitDegrees=45
  swing2LimitDegrees=45
  twistLimitDegrees=10
  twistMotion=Locked
 }
 {bone1Name=dyn_R_hairStrand_1, bone2Name=head, swing1LimitDegrees=60, swing2LimitDegrees=45, twistLimitDegrees=10, twistMotion=Locked}
 {bone1Name=dyn_R_hairStrand_2, bone2Name=dyn_R_hairStrand_1, swing1LimitDegrees=20, swing2LimitDegrees=45, twistLimitDegrees=10}
 {
  bone1Name=dyn_R_hairStrand_3
  bone2Name=dyn_R_hairStrand_2
  swing1LimitDegrees=20
  swing2LimitDegrees=20
  twistLimitDegrees=10
  twistMotion=Locked
 }
 {bone1Name=dyn_hairMass_4, bone2Name=dyn_hairMass_3, swing1LimitDegrees=80, swing2LimitDegrees=10, twistLimitDegrees=0}
 {
  bone1Name=spine_05
  bone2Name=spine_04
  swing1LimitDegrees=45
  swing1Motion=Free
  swing2LimitDegrees=45
  swing2Motion=Free
  twistLimitDegrees=45
  twistMotion=Free
 }
 {
  bone1Name=clavicle_r
  bone2Name=spine_05
  swing1LimitDegrees=45
  swing1Motion=Free
  swing2LimitDegrees=45
  swing2Motion=Free
  twistLimitDegrees=45
  twistMotion=Free
 }
 {
  bone1Name=clavicle_l
  bone2Name=spine_05
  swing1LimitDegrees=45
  swing1Motion=Free
  swing2LimitDegrees=45
  swing2Motion=Free
  twistLimitDegrees=45
  twistMotion=Free
 }
 {bone1Name=neck_02, bone2Name=spine_05, swing1LimitDegrees=45, swing2LimitDegrees=45, twistLimitDegrees=45}
 {
  bone1Name=spine_04
  bone2Name=spine_03
  swing1LimitDegrees=45
  swing1Motion=Free
  swing2LimitDegrees=45
  swing2Motion=Free
  twistLimitDegrees=45
  twistMotion=Free
 }
 {bone1Name=dyn_hairMass_base, bone2Name=head, swing1LimitDegrees=80, swing2LimitDegrees=10, twistLimitDegrees=0}
 {
  bone1Name=head
  bone2Name=neck_02
  swing1LimitDegrees=45
  swing1Motion=Free
  swing2LimitDegrees=45
  swing2Motion=Free
  twistLimitDegrees=45
  twistMotion=Free
 }
 {
  bone1Name=upperarm_l
  bone2Name=clavicle_l
  swing1LimitDegrees=45
  swing1Motion=Free
  swing2LimitDegrees=45
  swing2Motion=Free
  twistLimitDegrees=45
  twistMotion=Free
 }
 {
  bone1Name=upperarm_r
  bone2Name=clavicle_r
  swing1LimitDegrees=45
  swing1Motion=Free
  swing2LimitDegrees=45
  swing2Motion=Free
  twistLimitDegrees=45
  twistMotion=Free
 }
]
