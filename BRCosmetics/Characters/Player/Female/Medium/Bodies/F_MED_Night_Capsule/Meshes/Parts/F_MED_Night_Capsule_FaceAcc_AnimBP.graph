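# Post-process style secondary-motion AnimBP for the F_MED_Night_Capsule face accessory part of a player character.
# Copies the attached parent's pose and curves, layers Body_Mesh's pose from both upper arms, then runs RigidBody simulation in spine_04 space.
# Each update sets movement-state gravity for the simulation through a deprecated helper.
# A preventHairFlip flag for fast non-skydiving falls drives two hair-bone rotations, but that chain is not wired to the output pose.
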AnimBlueprint /BRCosmetics/Characters/Player/Female/Medium/Bodies/F_MED_Night_Capsule/Meshes/Parts/F_MED_Night_Capsule_FaceAcc_AnimBP
target_skeleton=&/F_MED_Night_Capsule_FaceAcc_Skeleton
graphs [
 AnimGraph {nodes [
   n1 = Get preventHairFlip() -> (preventHairFlip:double)
   n2 = "Transform (Modify) Bone - Bone: dyn_hairMass_1"(Alpha=@n1, Translation="0,0,0", Rotation="0,0,-50", Scale="1,1,1") {
    rotationMode=BMM_Additive
    alphaBoolBlend {blendOption=Linear}
    alphaScaleBiasClamp {bInterpResult=true, interpSpeedDecreasing=3, interpSpeedIncreasing=3}
    boneToModify {boneName=dyn_hairMass_1}
    rotation {roll=-50}
   }
   "Transform (Modify) Bone - Bone: dyn_hairMass_2"(ComponentPose=@n2, Alpha=@n1, Translation="0,0,0", Rotation="0,0,-50", Scale="1,1,1") {
    rotationMode=BMM_Additive
    alphaBoolBlend {blendOption=Linear}
    alphaScaleBiasClamp {bInterpResult=true, interpSpeedDecreasing=3, interpSpeedIncreasing=3}
    boneToModify {boneName=dyn_hairMass_2}
    rotation {roll=-50}
   }
   n3 = Copy Pose From Mesh(SourceMeshComponent=Get Body_Mesh())
   n4 = Copy Pose From Mesh() {bCopyCurves=true, bUseAttachedParent=true}
   n5 = RigidBody!(
    ComponentPose=Local To Component(LocalPose=Layered blend per bone(BasePose=@n4, BlendPoses_0=@n3) {layerSetup [{branchFilters [{boneName=upperarm_l}, {boneName=upperarm_r}]}]})
    Alpha
    OverrideWorldGravity:Vector=Get Gravity Override()
    ExternalForce
   ) -> (Pose) {
    bClampLinearTranslationLimitToRefPose=true
    bOverrideWorldGravity=true
    overridePhysicsAsset=&/F_MED_Night_Capsule_FaceAcc_Physics
    simulationSpace=BaseBoneSpace
    alphaBoolBlend {blendOption=Linear}
    baseBoneRef {boneName=spine_04}
    componentAppliedLinearAccClamp=<1000 1000 1000>
    componentLinearAccScale=<0.5 0.3 0.3>
    componentLinearVelScale=<0.3 0.1 0.1>
    overrideWorldGravity {z=-1300}
    simSpaceSettings {maxAngularAcceleration=1000, maxAngularVelocity=3, maxLinearAcceleration=50, maxLinearVelocity=500, velocityScaleZ=0, worldAlpha=1}
   }
   Output Pose(Result=Component To Local(ComponentPose=@n5)) {layerGroup=None}
  ]}
 EventGraph {nodes [
   Event BlueprintUpdateAnimation()
   > n1 = Sequence()
   > n2 = Gravity Override State Function DEPRECATED(CharacterPartAnimInstance=Self-Reference(), GameTimeInSeconds=GetGameTimeInSeconds(), GravityOverrideParams=Get Gravity Override Params())
   > Set Gravity Override(Gravity Override=@n2)
   n3 = NOT Boolean(A=Get bIsSkyDiving())
   n4 = "float < float"(A=Get PawnVelocityZ(), B=-1100)
   Set preventHairFlip(execute=@n1.then_1, preventHairFlip:double="To Float (Boolean)"(InBool=AND Boolean(A=@n3, B=@n4))) -> (Output_Get:double)
   Event BlueprintInitializeAnimation()
   > Set Body_Mesh(Body_Mesh=GetOwnerSkeletalMeshComponent())
   > n5 = Cast To PlayerPawn_Athena(Object=TryGetPawnOwner())
   > n6 = Set Pawn Owner(Pawn Owner:Player Pawn Athena Object Reference=@n5) -> (Output_Get:Player Pawn Athena Object Reference)
   > Set Head_Mesh(Head_Mesh:Skeletal Mesh Component Object Reference=GetSkeletalMeshForPartType(self=@n6, PartType=Head)) -> (Output_Get:Skeletal Mesh Component Object Reference)
  ]}
]
variables [
 Body_Mesh {type=SkeletalMeshComponent}
 Gravity Override {type=Vector}
 Gravity Override Params {
  type=GravityOverrideParamsStruct
  default {
   onGround_Moving {gravityOverride=<0 0 -1300>}
   onGround_Standing {gravityOverride=<0 0 -1300>}
   skydive_FloatingDown {gravityOverride=<0 -1000 -1100>, windAmplitude=<50 -125 100>}
   skydive_GoingUp {gravityOverride=<0 -300 -900>}
   skydive_Parachuting {gravityOverride=<0 100 -1300>, windAmplitude=<100 -250 200>, windFrequency=<1.5 2 4>}
  }
 }
 Head_Mesh {type=SkeletalMeshComponent}
 Pawn Owner {type=PlayerPawn_Athena_C}
 preventHairFlip {type=number}
 Rotation {type=Rotator, default=<rot 0 0 -45>}
]
